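# UI translucent material that draws the Vane Hush boss portrait inside a double-ringed circle, for the progress bar.
# Distortion-driven, scaled UVs feed two circle functions whose fills and outlines build the backdrop and a tintable outline color.
# The portrait samples with separately offset and scaled UVs, squares its RGB and composites into the final color.
# Circle shapes at their own opacity add to the icon alpha, scaled by master opacity, forming the mask.
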Material /SilentNovel_VaneHush/Gameplay/ProgressBar/Materials/M_UI_VaneHushBoss_Portrait
graph {
 properties {bAutomaticallySetUsageInEditor=false, blendMode=BLEND_Translucent, materialDomain=MD_UI, usedShadingModels=Default Lit}
 nodes [
  n1 = MaterialFunctionCall(UVs=NamedRerouteUsage(), Size=Constant() {r=0.52}, OutlineThickness=Constant() {r=0.025}) {materialFunction=&MF_UI_Circle}
  NamedRerouteDeclaration(None=@n1.Fill) {name=Inner Circle, nodeColor=<rgba 0 0 0 1>}
  NamedRerouteDeclaration(None=@n1.Outline) {name=Inner Outline, nodeColor=<rgba 0 0 0 1>}
  n2 = MaterialFunctionCall(UVs=NamedRerouteUsage(), Size=Constant() {r=0.6}, OutlineThickness=Constant() {r=0.06}) {materialFunction=&MF_UI_Circle}
  NamedRerouteDeclaration(None=@n2.Fill) {name=Outer Circle, nodeColor=<rgba 0 0 0 1>}
  n3 = TextureCoordinate()
  n4 = Add(A=ScalarParameter() {defaultValue=-0.01, parameterName=Icon X pos}, B=ComponentMask(None=@n3) {r=true}) {constB=0}
  n5 = Add(A=ComponentMask(None=@n3) {g=true}, B=ScalarParameter() {defaultValue=0.06, parameterName=Icon Y pos}) {constB=0.06}
  n6 = TextureSample(UVs=MaterialFunctionCall(UVs=AppendVector(A=@n4, B=@n5), Texture Scale=ScalarParameter() {defaultValue=0.6, parameterName=Icon Scale}) {materialFunction=&ScaleUVsByCenter}.UVs) {samplerSource=SSM_Clamp_WorldGroupSettings, texture=/SilentNovel_VaneHush/Gameplay/ProgressBar/Textures/T_UI_SilentNovel_VaneHush_Portrait}
  NamedRerouteDeclaration(None=Multiply(A=@n6.RGB, B=@n6.RGB)) {name=IconColor, nodeColor=<rgba 0 0 0 1>}
  NamedRerouteDeclaration(None=@n6.A) {name=IconMask, nodeColor=<rgba 0 0 0 1>}
  n7 = Add(
   A=Subtract(A=TextureCoordinate()) {constB=0.5}
   B=MaterialFunctionCall!(UV, SpeedX=Constant(), SpeedY=Constant(), Intensity=ScalarParameter() {defaultValue=0.06, parameterName=Deformation intensity}) {materialFunction=&/MF_UI_WildEstate_Distort_2}.Result
  )
  n8 = TextureCoordinate()
  n9 = AppendVector(A=Add(B=ComponentMask(None=@n8) {r=true}) {constB=0}, B=Add(A=ComponentMask(None=@n8) {g=true}) {constB=-0.15})
  NamedRerouteDeclaration(None=MaterialFunctionCall(UVs=Add(A=@n7, B=@n9), Texture Scale=ScalarParameter() {defaultValue=1.5, parameterName=Circle scale}) {materialFunction=&ScaleUVsByCenter}.UVs) {name=Circle UV, nodeColor=<rgba 0 0 0 1>}
  n10 = NamedRerouteUsage()
  n11 = NamedRerouteUsage()
  n12 = NamedRerouteUsage()
  n13 = NamedRerouteUsage()
  n14 = Multiply(
   A=ScalarParameter() {defaultValue=0.8, parameterName=Circle Opacity}
   B=Add(A=LinearInterpolate(A=@n10, B=@n11, Alpha=@n11), B=LinearInterpolate(A=@n12, B=@n13, Alpha=@n13))
  ) {constA=0.8}
  n15 = NamedRerouteUsage()
  NamedRerouteDeclaration(None=Multiply(A=Add(A=@n14, B=@n15), B=ScalarParameter() {defaultValue=1, parameterName=Master Opacity})) {name=Mask, nodeColor=<rgba 0 0 0 1>}
  n16 = LinearInterpolate(
   A=LinearInterpolate(B=VectorParameter() {parameterName=OutlineColor, defaultValue=<rgba 0.5583 0.0152 0.0203 1>}.RGB, Alpha=NamedRerouteUsage())
   Alpha=NamedRerouteUsage()
  )
  n17 = Constant3Vector()
  n18 = NamedRerouteUsage()
  NamedRerouteDeclaration(None=LinearInterpolate(A=LinearInterpolate(A=@n16, B=@n17, Alpha=@n18), B=NamedRerouteUsage(), Alpha=NamedRerouteUsage())) {name=Final Color, nodeColor=<rgba 0 0 0 1>}
  NamedRerouteDeclaration(None=@n2.Outline) {name=Outer Outline, nodeColor=<rgba 0 0 0 1>}
  NamedRerouteUsage()
  NamedRerouteUsage()
 ]
}
