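# Translucent UI material for the VaneHush progress bar, drawing confetti sprinkles and a donut icon that shifts with the Progress parameter.
# Aspect-corrected UVs form cells, with sine-dot hashes randomizing values such as fill chance and palette color.
# Each sprinkle is a rotated, rounded MF_UI_Box with fill and outline.
# A hard progress step and the icon texture's alpha provide masks for compositing color and opacity.
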Material /SilentNovel_VaneHush/Gameplay/ProgressBar/Materials/M_UI_VaneHush_Confetti
graph {
 properties {bAutomaticallySetUsageInEditor=false, blendMode=BLEND_Translucent, materialDomain=MD_UI, usedShadingModels=Default Lit}
 nodes [
  n1 = MaterialFunctionCall!(A=NamedRerouteUsage(), B=ScalarParameter() {defaultValue=-0.1, parameterName=FillChance}) {materialFunction=/FortUILibrary/MaterialFunctions/Basics/MF_UI_BooleanOperations}
  n2 = NamedRerouteUsage()
  n3 = LinearInterpolate(A=@n2, B=@n2, Alpha=NamedRerouteUsage())
  n4 = Multiply(A=Multiply(A=ScalarParameter() {defaultValue=1, parameterName=Sprinkles Opacity}, B=Multiply(A=@n1.IsGreater, B=@n3)), B=NamedRerouteUsage())
  n5 = NamedRerouteUsage()
  NamedRerouteDeclaration(None=Multiply(A=Add(A=@n4, B=@n5), B=ScalarParameter() {defaultValue=1, parameterName=Master Opacity})) {name=Mask, nodeColor=<rgba 0 0 0 1>}
  n6 = NamedRerouteUsage()
  n7 = Sine(None=DotProduct(A=Floor(None=@n6), B=Constant2Vector() {g=45.5, r=12.2}))
  n8 = Constant() {r=0.2}
  NamedRerouteDeclaration(None=Frac(None=Multiply(A=@n7, B=@n8))) {name=Random, nodeColor=<rgba 0 0.2031 1 1>}
  n9 = NamedRerouteUsage()
  n10 = LinearInterpolate(
   A=VectorParameter() {parameterName=Col1, defaultValue=<rgba 1 0.9383 0.0851 1>}.RGB
   B=Constant3Vector() {constant=<rgba 0.5776 0.1655 0.1655 1>}
   Alpha=Floor(None=Multiply(A=@n9) {constB=3})
  )
  n11 = VectorParameter() {parameterName=Col2, defaultValue=<rgba 0.0052 0.2747 0.0953 1>}
  n12 = NamedRerouteUsage()
  n13 = Step(X=Floor(None=Multiply(A=@n12) {constB=8})) {constY=1}
  n14 = VectorParameter() {parameterName=Col3, defaultValue=<rgba 0.9647 0.2035 0.208 1>}
  n15 = VectorParameter() {parameterName=Col4, defaultValue=<rgba 1 1 1 1>}
  n16 = NamedRerouteUsage()
  n17 = Step(X=Floor(None=Multiply(A=@n16) {constB=1.2})) {constY=1}
  n18 = Floor(None=Multiply(A=NamedRerouteUsage()) {constB=5})
  n19 = LinearInterpolate(
   A=LinearInterpolate(A=LinearInterpolate(A=@n10, B=@n11.RGB, Alpha=@n13), B=LinearInterpolate(A=@n14.RGB, B=@n15.RGB, Alpha=@n17), Alpha=Step(X=@n18) {constY=4})
   Alpha=NamedRerouteUsage()
  ) {constB=-5}
  n20 = Constant3Vector()
  n21 = NamedRerouteUsage()
  NamedRerouteDeclaration(None=LinearInterpolate(A=LinearInterpolate(A=@n19, B=@n20, Alpha=@n21) {constB=0}, B=NamedRerouteUsage(), Alpha=NamedRerouteUsage()) {constAlpha=1}) {name=Color, nodeColor=<rgba 0 0 0 1>}
  NamedRerouteDeclaration(None=Divide(A=ScalarParameter() {defaultValue=640, parameterName=Width}, B=ScalarParameter() {defaultValue=48, parameterName=Height})) {name=AspectRatio, nodeColor=<rgba 1 0 0.6094 1>}
  n22 = TextureCoordinate()
  n23 = ComponentMask(None=@n22) {r=true}
  n24 = NamedRerouteUsage()
  NamedRerouteDeclaration(None=AppendVector(A=Multiply(A=@n23, B=@n24), B=ComponentMask(None=@n22) {g=true})) {name=UV_Fixed, nodeColor=<rgba 0 0.9297 1 1>}
  n25 = NamedRerouteUsage()
  n26 = Sine(None=DotProduct(A=Floor(None=@n25), B=Constant2Vector() {g=13.1, r=8.3}))
  n27 = Constant() {r=0.5}
  NamedRerouteDeclaration(None=Frac(None=Multiply(A=@n26, B=@n27))) {name=Random2, nodeColor=<rgba 0 0.2031 1 1>}
  n28 = Subtract(
   A=ComponentMask(None=MaterialFunctionCall() {materialFunction=&GetUserInterfaceUV}.Normalized UV) {r=true}
   B=ScalarParameter() {defaultValue=1, group=Progress, parameterName=Progress, sliderMax=1}
  )
  NamedRerouteDeclaration(None=OneMinus(None=Step(X=@n28) {constX=0})) {name=ProgressGradient, nodeColor=<rgba 0 0 0 1>}
  n29 = Frac(None=NamedRerouteUsage())
  n30 = NamedRerouteUsage()
  n31 = MaterialFunctionCall(
   UVs=Panner(
    Coordinate=MaterialFunctionCall(UVs=@n29, "Rotation Angle (0-1)"=@n30) {materialFunction=&CustomRotator}.Rotated Values
    Time=Constant() {r=0.8}
   ) {speedX=0.1}
   Size=Constant2Vector() {g=0.2, r=0.1}
   CornerRoundness=Constant() {r=0.1}
   OutlineThickness=Constant() {r=0.04}
  ) {materialFunction=&MF_UI_Box}
  NamedRerouteDeclaration(None=@n31.Fill) {name=Fill, nodeColor=<rgba 0 0 0 1>}
  NamedRerouteDeclaration(None=@n31.Outline) {name=Outline, nodeColor=<rgba 0 0 0 1>}
  n32 = NamedRerouteUsage()
  n33 = ScalarParameter() {defaultValue=1, group=Progress, parameterName=Progress, sliderMax=1}
  n34 = Subtract(A=NamedRerouteUsage())
  n35 = Add(A=ComponentMask(None=@n32) {r=true}, B=Multiply(A=Multiply(A=@n33, B=@n34)) {constB=-1}) {constB=-23.6}
  n36 = ComponentMask(None=@n32) {g=true}
  n37 = TextureSample(UVs=MaterialFunctionCall(UVs=AppendVector(A=@n35, B=@n36), Texture Scale=ScalarParameter() {defaultValue=1.3, parameterName=Icon scale}) {materialFunction=&ScaleUVsByCenter}.UVs) {samplerSource=SSM_Clamp_WorldGroupSettings, texture=/SilentNovel_VaneHush/Gameplay/ProgressBar/Textures/T_UI_SilentNovel_VaneHush_Icon}
  NamedRerouteDeclaration(None=Multiply(A=@n37.RGB, B=@n37.RGB)) {name=DonutColor, nodeColor=<rgba 0 0 0 1>}
  NamedRerouteDeclaration(None=@n37.A) {name=Donut Mask, nodeColor=<rgba 0 0 0 1>}
  NamedRerouteUsage()
  NamedRerouteUsage()
 ]
}
